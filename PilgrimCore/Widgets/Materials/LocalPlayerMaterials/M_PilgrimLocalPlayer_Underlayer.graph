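# Translucent UI material for the local player underlayer, apparently behind a score, with a two-colour horizontal gradient that can be inverted, offset and eased.
# A rotated smoothstep forms an angled underlayer mask, alongside a rounded-box shape mask and an opacity control.
# An optional triangle pattern, overlay-blended for colour, pans in screen-aligned UVs, cycles its RGB channels over time with a sine and fades along a gradient.
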Material /PilgrimCore/Widgets/Materials/LocalPlayerMaterials/M_PilgrimLocalPlayer_Underlayer
graph {
 properties {
  bAutomaticallySetUsageInEditor=false
  bUsedWithStaticMesh=false
  blendMode=BLEND_Translucent
  materialDomain=MD_UI
  pixelDepthOffsetMode=PDOM_Legacy
  usedShadingModels=Default Lit
 }
 nodes [
  MaterialFunctionCall(Size=Constant() {r=1}, CornerRoundness=Constant() {r=1}, KeepAspectRatio=StaticBool() {value=true}) {materialFunction=&MF_UI_Box}
  MaterialFunctionCall() {materialFunction=&MF_UI_UVScaling}
  NamedRerouteDeclaration(None=MaterialFunctionCall(Corner Radius=Constant() {r=1}, Radius in Pixels=Constant()) {materialFunction=&MF_UI_CoreUI_RoundedMask}.Shape Mask) {name=Main_Shape, nodeColor=<rgba 0.4531 0 1 1>}
  n1 = MaterialFunctionCall(Base=NamedRerouteUsage(), Blend=NamedRerouteUsage()) {materialFunction=&Blend_Overlay}
  n2 = NamedRerouteUsage()
  n3 = ScalarParameter() {defaultValue=0.5, parameterName=Triangles_Op}
  NamedRerouteDeclaration(None=StaticSwitchParameter(True=LinearInterpolate(A=@n1.Result, B=@n2, Alpha=@n3), False=@n2) {defaultValue=true, parameterName=Triangle_Pattern}) {name=Base_Colour, nodeColor=<rgba 0 1 0.1328 1>}
  n4 = ScalarParameter() {defaultValue=0.612, parameterName=AnglePosition}
  n5 = ScalarParameter() {defaultValue=0.005, parameterName=AngleSoftness}
  n6 = MaterialFunctionCall(UVs=TextureCoordinate(), "Rotation Angle (0-1)"=ScalarParameter() {defaultValue=-0.01, parameterName=AdjustAngle}) {materialFunction=&CustomRotator}
  NamedRerouteDeclaration(None=SmoothStep(Min=@n4, Max=Add(A=@n4, B=@n5), Value=ComponentMask(None=@n6.Rotated Values) {r=true}) {constMax=0.605, constMin=0.6}) {name=ScoreUnderlayer_Mask, nodeColor=<rgba 1 0 0.1875 1>}
  n7 = VectorParameter() {parameterName=UnderLayer_GR1, defaultValue=<rgba 0.9734 0 0.2392 1>}
  n8 = VectorParameter() {parameterName=UnderLayer_GR2, defaultValue=<rgba 0.9911 0.8702 0 1>}
  n9 = Add(
   A=ComponentMask(None=MaterialFunctionCall() {materialFunction=&GetUserInterfaceUV}.Normalized UV) {r=true}
   B=AppendVector(A=ScalarParameter() {parameterName=OffsetGradient_X}, B=Constant())
  )
  n10 = LinearInterpolate(
   A=@n9
   B=MaterialFunctionCall("Value (0-1)"=@n9) {materialFunction=&MF_UI_EaseCurves}.EaseOutQuint
   Alpha=ScalarParameter() {parameterName=EaseCurve_Amount}
  )
  NamedRerouteDeclaration(None=StaticSwitchParameter(True=LinearInterpolate(A=@n7.RGB, B=@n8.RGB, Alpha=@n10), False=LinearInterpolate(A=@n8.RGB, B=@n7.RGB, Alpha=@n10)) {parameterName=InvertGradient}) {name=UnderLayer_Base_Colour, nodeColor=<rgba 0 0.5078 1 1>}
  NamedRerouteDeclaration(None=@n8.RGB) {name=Blend_Colour, nodeColor=<rgba 0 0.6016 1 1>}
  n11 = NamedRerouteDeclaration(None=MaterialFunctionCall() {materialFunction=&ScreenAlignedUVs}."X 100%, Y 100%") {name=ScreenSizeUV, nodeColor=<rgba 0 0 0 1>}
  n12 = AppendVector(A=Constant() {r=14}, B=Constant() {r=8})
  n13 = TextureSample(UVs=Panner(Coordinate=Multiply(A=@n11, B=@n12), Speed=Constant() {r=0.005})) {samplerType=SAMPLERTYPE_LinearColor, texture=/Game/UI/Foundation/Textures/Pattern/T_Triangle_BattlePassSmall}
  n14 = SmoothStep(Value=@n13.R) {constMax=0.2, constMin=0.08}
  n15 = SmoothStep(Value=@n13.G) {constMax=0.2, constMin=0.08}
  n16 = SmoothStep(Value=@n13.B) {constMax=0.2, constMin=0.08}
  n17 = MaterialFunctionCall(Speed=Constant() {r=0.006}) {materialFunction=&DebugTimeSine}
  n18 = ScalarParameter() {defaultValue=0.25, parameterName=Gradient Fade Min}
  n19 = ScalarParameter() {defaultValue=1.2, parameterName=Gradient Fade Max}
  n20 = OneMinus(None=MaterialFunctionCall() {materialFunction=&LinearGradient}.UGradient)
  NamedRerouteDeclaration(None=Min(
    A=MaterialFunctionCall(1=@n14, 2=@n15, 3=@n16, A=@n17.Result) {materialFunction=&Lerp_Multiple_Float3}.Lerp 3 Inputs
    B=SmoothStep(Min=@n18, Max=@n19, Value=@n20) {constMax=1.2, constMin=0.25}
   )) {name=Triangle_Pattern, nodeColor=<rgba 1 0 0.8437 1>}
  NamedRerouteDeclaration(None=@n2) {name=Colour_Mixer, nodeColor=<rgba 0 1 0.2734 1>}
  NamedRerouteUsage()
  Saturate(None=LinearInterpolate(B=NamedRerouteUsage(), Alpha=ScalarParameter() {defaultValue=1, parameterName=Opacity Settings}) {constAlpha=1})
 ]
}
