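# Translucent UI material for local multiplayer player icons, built on the T_Ui_Lead_Icon texture.
# A Size parameter scales the UVs around their center before the texture is sampled.
# The inverted red channel is treated as a distance field and passed through MF_UI_AntiAliasing with a fixed threshold and smoothing to give smooth icon edges.
# A Color parameter, white by default, is exposed alongside the shape mask.
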Material /PilgrimCore/Widgets/Materials/LocalPlayerMaterials/M_Ui_LocalMultiplayer_Icons
graph {
 properties {
  bAutomaticallySetUsageInEditor=false
  bUsedWithStaticMesh=false
  blendMode=BLEND_Translucent
  materialDomain=MD_UI
  usedShadingModels=Default Lit
 }
 nodes [
  n1 = MaterialFunctionCall(UVs=TextureCoordinate(), Texture Scale=ScalarParameter() {defaultValue=1, parameterName=Size}) {materialFunction=&ScaleUVsByCenter}
  MaterialFunctionCall(
   DistanceField=OneMinus(None=TextureSample(UVs=@n1.UVs) {samplerType=SAMPLERTYPE_LinearColor, texture=/Game/UI/Icons/Icon-Nodes/T_Ui_Lead_Icon}.R)
   Threshold=Constant() {r=0.5}
   Smoothing=Constant() {r=0.5}
  ) {materialFunction=&MF_UI_AntiAliasing}
  VectorParameter() {parameterName=Color, defaultValue=<rgba 1 1 1 1>}
 ]
}
